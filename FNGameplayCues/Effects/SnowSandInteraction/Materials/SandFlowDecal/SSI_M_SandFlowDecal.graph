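# Deferred decal material for Niagara mesh particles that draws animated sand flow around footprints in the snow and sand interaction effects.
# Dynamic parameters scale and rotate particle-relative decal UVs, driving sine flow bands distorted by a sand normal.
# A footprint texture, radial and slope fades and a custom depth test mask the opacity, which is zero on forward and mobile paths.
# Landscape runtime virtual textures supply a normal to blend with and a height-based vertical offset.
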Material /FNGameplayCues/Effects/SnowSandInteraction/Materials/SandFlowDecal/SSI_M_SandFlowDecal
graph {
 properties {
  bUsedWithNiagaraMeshParticles=true
  blendMode=BLEND_Translucent
  materialDomain=MD_DeferredDecal
  pixelDepthOffsetMode=PDOM_Legacy
  usedShadingModels=Default Lit
 }
 nodes [
  n1 = DynamicParameter() {parameterIndex=2, defaultValue=<rgba 0 0 0 1>, paramNames [Random, FlowScale, "", ""]}
  n2 = MaterialFunctionCall() {materialFunction=&WorldPositionBehindTranslucency}
  n3 = ParticlePositionWS()
  n4 = ActorPositionWS()
  n5 = ScalarParameter() {parameterName=N}
  n6 = Divide(A=Subtract(A=@n2.Result, B=LinearInterpolate(A=@n3, B=@n4, Alpha=@n5)), B=Multiply(A=@n1.FlowScale) {constB=100}) {constB=100}
  n7 = DynamicParameter() {parameterIndex=3, defaultValue=<rgba 1 1 0 0>, paramNames [QuatX,QuatY,QuatZ,QuatW]}
  n8 = MaterialFunctionCall(R=@n7.QuatX, G=@n7.QuatX, B=@n7.QuatX, A=@n7.QuatX) {materialFunction=&AppendMany}
  n9 = StaticBool() {value=true}
  n10 = ComponentMask(None=Add(A=MaterialFunctionCall!(Vector=@n6, Quat=@n8.RGBA, Invert Quaternion=@n9) {materialFunction=/Engine/Functions/Engine_MaterialFunctions03/Math/MultiplyVectorWithQuaternion}.Result) {constB=0.5}) {g=true, r=true}
  n11 = VectorParameter() {parameterName=OFfset, defaultValue=<rgba 0 0 0 1>}
  n12 = DynamicParameter() {defaultValue=<rgba 1 1 0 0>, paramNames [CompressedQuaternion, BlurMarkScale, FlowNormalRotation, FlowNormalSlope]}
  n13 = MaterialFunctionCall(
   UVs=Add(A=@n10, B=AppendVector(A=@n11.R, B=@n11.R))
   Rotation Center=Constant2Vector() {g=0.5, r=0.5}
   "Rotation Angle (0-1)"=Add(A=@n12.FlowNormalRotation, B=ScalarParameter() {parameterName=FlowNormal_Rotation})
  ) {materialFunction=&CustomRotator}
  n14 = VectorParameter() {parameterName=DistortScale, defaultValue=<rgba 0 0 0 1>}
  n15 = AppendVector(A=@n14.R, B=@n14.R)
  n16 = TextureSampleParameter2D(UVs=Add(
    A=@n1.Random
    B=MaterialFunctionCall(UVs=@n13.Rotated Values, Texture Scale=@n15) {materialFunction=&ScaleUVsByCenter}.UVs
   )) {parameterName=DistortNormal, samplerType=SAMPLERTYPE_Normal, texture=/Game/Environments/Sets/Arid/Textures/DesertSand_normal}
  n17 = AppendVector(A=@n16.R, B=@n16.R)
  n18 = ScalarParameter() {parameterName=Distort}
  n19 = ScalarParameter() {parameterName=FlowNormal_Slope}
  n20 = Add(
   A=@n12.FlowNormalSlope
   B=LinearInterpolate(B=ScalarParameter() {defaultValue=0.15, parameterName=FlowNormal_SlopeMaxOffset}, Alpha=@n19) {constB=0.1}
  )
  n21 = Add(A=@n13.Rotated Values, B=AppendVector(A=Constant(), B=Multiply(A=@n20) {constB=-1}))
  n22 = VectorParameter() {parameterName=FlowNormal_UVScale_Flat, defaultValue=<rgba 1 1 0 1>}
  n23 = VectorParameter() {parameterName=FlowNormal_UVScale_Slope, defaultValue=<rgba 1 2 0 1>}
  n24 = LinearInterpolate(A=AppendVector(A=@n22.R, B=@n22.R), B=AppendVector(A=@n23.R, B=@n23.R), Alpha=@n19)
  n25 = Distance(
   A=Add(
    A=Multiply(A=@n17, B=@n18)
    B=MaterialFunctionCall(UVs=@n21, Texture Scale=@n24) {materialFunction=&ScaleUVsByCenter}.UVs
   )
   B=Constant() {r=0.5}
  )
  n26 = DynamicParameter() {parameterIndex=1, defaultValue=<rgba 0 0 0 1>, paramNames [FlowNormalFlat, "", FlowNormalTime, FootPrintSize]}
  Add(A=@n25, B=Multiply(A=Add(A=@n26.FlowNormalTime, B=@n1.Random), B=ScalarParameter() {defaultValue=-0.07, parameterName=FlowNormal_FlowSpeed}))
  n27 = DynamicParameter() {defaultValue=<rgba 1 1 0 0>, paramNames [CompressedQuaternion, BlurMarkScale, FlowNormalRotation, FlowNormalSlope]}
  MaterialFunctionCall!(CompressedQuaternion=@n27.CompressedQuaternion) {materialFunction=/Niagara/Functions/DataPacking/Niagara_DecompressQuaternion}
  n28 = TextureSample() {texture=/Game/Effects/Fort_Effects/Gameplay/Tutorial_Target/shader_mask_round_square_blurry}
  n29 = MaterialFunctionCall(Float2=@n13.Rotated Values) {materialFunction=&BreakOutFloat2Components}
  n30 = ScalarParameter() {parameterName=FlowNormal_Fade}
  n31 = Multiply(
   A=MaterialFunctionCall(UVs=@n13.Rotated Values, Density=ScalarParameter() {parameterName=FlowNormal_EdgeDensity}) {materialFunction=&RadialGradientExponential}.RadialGradientExponential
   B=Power(Base=OneMinus(None=@n29.G), Exp=LinearInterpolate(B=@n30, Alpha=@n19))
  )
  n32 = ScalarParameter() {parameterName=Fade2Offset}
  n33 = Add(A=@n32)
  n34 = Saturate(None=Multiply(A=Add(A=@n29.G, B=@n32), B=Divide(B=@n33)))
  n35 = ScalarParameter() {parameterName=FlowNormal_Fade2}
  n36 = DynamicParameter() {parameterIndex=1, defaultValue=<rgba 0 0 0 1>, paramNames [FlowNormalFlat, "", FlowNormalTime, FootPrintSize]}
  n37 = Multiply(A=OneMinus(None=Power(Base=@n34, Exp=@n35)), B=OneMinus(None=Power(Base=@n36.FlowNormalFlat)))
  n38 = ScalarParameter() {defaultValue=1, parameterName=FlowNormal_Fade2_AreaMul}
  n39 = Multiply(A=@n26.FlowNormalTime, B=ScalarParameter() {defaultValue=-1.25, parameterName=TestFC})
  n40 = Add(
   A=Multiply(A=Add(A=@n25, B=@n39), B=ScalarParameter() {defaultValue=4, parameterName=TestFM})
   B=ScalarParameter() {defaultValue=1.2, parameterName=TestFO}
  ) {constB=0.3}
  n41 = Multiply(A=@n40) {constB=0.5}
  n42 = MaterialFunctionCall!(
   TexCoord=MaterialFunctionCall(X=Constant() {r=0.5}, Y=@n40) {materialFunction=&MakeFloat2}.Result
   ClampChannel=Constant2Vector() {g=1, r=1}
  ) {materialFunction=/Game/Effects/Fort_Effects/MaterialFunctions/MF_FX_TilingMethod}
  n43 = Multiply(A=Multiply(A=Sine(None=@n41), B=@n42.Result), B=Saturate(None=Multiply(A=@n37, B=@n38)))
  n44 = LinearInterpolate(A=ScalarParameter() {parameterName=FlowNormal_Flat}, Alpha=@n36.FlowNormalFlat)
  n45 = Multiply(A=Multiply(A=@n31, B=@n43), B=OneMinus(None=@n44))
  n46 = TextureCoordinate()
  n47 = ScalarParameter() {defaultValue=-0.9, parameterName=Mask_EdgeWidth}
  n48 = Add(A=@n46, B=@n47) {constB=-0.95}
  n49 = Divide(B=Add(B=@n47) {constA=1, constB=-0.95})
  n50 = Add(A=OneMinus(None=@n46)) {constB=-0.95}
  n51 = MaterialFunctionCall(Float2=Max(A=Multiply(A=@n48, B=@n49) {constB=20}, B=Multiply(A=@n50) {constB=20})) {materialFunction=&BreakOutFloat2Components}
  n52 = ScreenPosition()
  n53 = Floor(None=Subtract(A=SceneTexture(UVs=@n52.ViewportUV) {sceneTextureId=PPI_CustomDepth}.Color) {constB=4096})
  n54 = Multiply(
   A=OneMinus(None=Max(A=@n51.R, B=@n51.R))
   B=MaterialFunctionCall(Float3=Saturate(None=@n53)) {materialFunction=&BreakOutFloat3Components}.R
  )
  n55 = ScalarParameter() {defaultValue=1, parameterName=ShoeMark_BlurMark_MinScale}
  n56 = ScalarParameter() {defaultValue=1, parameterName=ShoeMark_BlurMark_MaxScale}
  n57 = DynamicParameter() {defaultValue=<rgba 1 1 0 0>, paramNames [CompressedQuaternion, BlurMarkScale, FlowNormalRotation, FlowNormalSlope]}
  n58 = DynamicParameter() {parameterIndex=1, defaultValue=<rgba 0 0 0 1>, paramNames [FlowNormalFlat, "", FlowNormalTime, FootPrintSize]}
  n59 = MaterialFunctionCall(UVs=@n10, Texture Scale=@n58.FootPrintSize) {materialFunction=&ScaleUVsByCenter}
  n60 = MaterialFunctionCall(TexCoord=@n59.UVs, ClampChannel=Constant() {r=1}) {materialFunction=&MF_FX_TilingMethod}
  n61 = TextureSample(UVs=MaterialFunctionCall(UVs=LinearInterpolate(B=@n59.UVs, Alpha=@n60.Result), Texture Scale=LinearInterpolate(A=@n55, B=@n56, Alpha=@n57.BlurMarkScale) {constA=1}) {materialFunction=&ScaleUVsByCenter}.UVs) {texture=/Game/Effects/SnowSandInteraction/Textures/SSI_T_FootprintMasks}
  n62 = Multiply(A=ParticleColor().A, B=Multiply(A=ScalarParameter(), B=Multiply(A=@n54, B=@n61.G)))
  n63 = Multiply(A=Saturate(None=Multiply(A=@n45) {constB=2}), B=@n62)
  n64 = MaterialFunctionCall(Float3=NamedRerouteUsage()) {materialFunction=&BreakOutFloat3Components}
  n65 = Abs(None=@n64.R)
  n66 = If(A=@n65, "A > B"=@n63, "A == B"=@n62, "A < B"=@n62) {constB=0.6}
  n67 = Abs(None=@n64.G)
  n68 = Saturate(None=If(A=@n67, "A > B"=@n63, "A == B"=@n66, "A < B"=@n66) {constB=0.6})
  n69 = Constant()
  Multiply(
   A=Saturate(None=Multiply(A=@n28.RGB) {constB=2})
   B=Multiply(A=NamedRerouteUsage(), B=ShadingPathSwitch(Default=@n68, Deferred=@n68, Forward=@n69, Mobile=@n69))
  )
  n70 = Transform(None=Constant3Vector() {constant=<rgba 0 0 1 0>})
  n71 = RuntimeVirtualTextureSampleParameter(World Position=@n2.Result) {
   group=RVT
   materialType=BaseColor_Normal_Specular_YCoCg
   parameterName=RVT Sample Target
   virtualTexture=/Game/Athena/Apollo/Maps/Landscape/VT/RVT_Apollo_Landscape
  }
  n72 = Abs(None=DotProduct(A=@n70, B=@n71.Normal))
  n73 = LinearInterpolate(A=@n70, Alpha=Power(Base=Saturate(None=@n72)) {constExponent=10})
  NamedRerouteDeclaration(None=@n73) {name=NormalMismatchMask, nodeColor=<rgba 1 0 0.9609 1>}
  n74 = Multiply(A=Add(A=@n21) {constB=-0.5}) {constB=2}
  n75 = Multiply(A=@n40, B=@n42.Result)
  n76 = Add(A=@n75, B=ScalarParameter() {parameterName=OffsetTTTE}) {constB=0.25}
  n77 = ScalarParameter() {defaultValue=2, parameterName=FlowNormal_Density}
  n78 = Add(A=Sine(None=Multiply(A=@n76, B=@n77) {constB=5}))
  n79 = Multiply(A=LinearInterpolate(A=@n74, B=Multiply(A=@n74) {constB=-1}, Alpha=Multiply(A=@n78) {constB=0.5}), B=ScalarParameter() {parameterName=TestD})
  n80 = DotProduct(A=@n79, B=@n79)
  n81 = AppendVector(A=@n79, B=OneMinus(None=Clamp(None=@n80)))
  Normalize(VectorInput=@n81)
  Saturate(None=@n75)
  ScalarParameter() {defaultValue=1, parameterName=POM_UVScale}
  n82 = Constant3Vector() {constant=<rgba 0 0 1 0>}
  n83 = LinearInterpolate(
   A=Constant3Vector() {constant=<rgba 0 0 1 0>}
   B=MaterialFunctionCall(Normal=Multiply(A=@n81) {constB=0.5}, Flatness=@n44) {materialFunction=&FlattenNormal}.Result
   Alpha=Multiply(A=@n61.G, B=@n31)
  )
  n84 = NamedRerouteDeclaration(None=@n73) {name=RVTNormal, nodeColor=<rgba 0 1 0.9766 1>}
  n85 = MaterialFunctionCall(
   BaseNormal=LinearInterpolate(
    A=MaterialFunctionCall(Vector=@n84, Quat=@n8.RGBA, Invert Quaternion=@n9) {materialFunction=&MultiplyVectorWithQuaternion}.Result
    B=Constant3Vector() {constant=<rgba 0 0 1 0>}
    Alpha=@n43
   )
   AdditionalNormal=@n83
  ) {materialFunction=&BlendAngleCorrectedNormals}
  n86 = If(A=@n65, "A > B"=@n83, "A == B"=@n85.Result, "A < B"=@n85.Result) {constB=0.6}
  n87 = If(A=@n67, "A > B"=@n83, "A == B"=@n86, "A < B"=@n86) {constB=0.6}
  ShadingPathSwitch(Default=@n87, Deferred=@n87, Forward=@n82, Mobile=@n82)
  n88 = Constant3Vector() {constant=<rgba 0 0 1 0>}
  n89 = WorldPosition() {worldPositionShaderOffset=WPT_ExcludeAllShaderOffsets}
  n90 = RuntimeVirtualTextureSampleParameter(World Position=@n89.XYZ) {
   group=RVT
   materialType=WorldHeight
   parameterName=RVT Sample Target_Height
   virtualTexture=/Game/Athena/Apollo/Maps/Landscape/VT/RVT_Apollo_Landscape_Height
  }
  n91 = MaterialFunctionCall(Z=Subtract(
    A=Add(A=@n90.WorldHeight) {constB=12}
    B=MaterialFunctionCall(Float3=@n89.XYZ) {materialFunction=&BreakOutFloat3Components}.B
   )) {materialFunction=&MakeFloat3}
  n92 = If(A=@n65, "A > B"=@n88, "A == B"=@n91.Result, "A < B"=@n91.Result) {constB=0.6}
  n93 = If(A=@n67, "A > B"=@n88, "A == B"=@n92, "A < B"=@n92) {constB=0.6}
  ShadingPathSwitch(Default=@n93, Deferred=@n93, Forward=@n88, Mobile=@n88)
 ]
}
